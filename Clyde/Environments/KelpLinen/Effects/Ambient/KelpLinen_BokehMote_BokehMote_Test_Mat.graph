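# Test additive, unlit Niagara sprite material for ambient KelpLinen bokeh motes, faking depth-of-field blur.
# Two opposed camera depth fades around a focal point form a tilt-shift mask.
# Reroutes feed color blending between two tints and scaling of offsets from particle center, with a dynamic-parameter size multiplier.
# A soft sphere mask times particle alpha is depth-faded for opacity, alongside near-camera fade and gradient nodes.
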Material /Clyde/Environments/KelpLinen/Effects/Ambient/KelpLinen_BokehMote_BokehMote_Test_Mat
graph {
 properties {
  bUsedWithNiagaraSprites=true
  blendMode=BLEND_Additive
  pixelDepthOffsetMode=PDOM_Legacy
  shadingModel=MSM_Unlit
  translucencyLightingMode=TLM_VolumetricPerVertexNonDirectional
  usedShadingModels=Unlit
 }
 nodes [
  LinearInterpolate(
   A=VectorParameter() {parameterName=ColorA, defaultValue=<rgba 1 0.006 0 1>}.RGB
   B=VectorParameter() {parameterName=ColorB, defaultValue=<rgba 0.0403 0 1 1>}.RGB
   Alpha=NamedRerouteUsage()
  )
  MaterialFunctionCall(Distance=ScalarParameter() {defaultValue=128, parameterName=NearFade}) {materialFunction=&Near_Camera_Fade}
  n1 = MaterialFunctionCall(UV Channel=TextureCoordinate()) {materialFunction=&LinearGradient}
  Multiply(A=Multiply(A=OneMinus(None=@n1.UGradient), B=Multiply(A=@n1.UGradient))) {constB=4}
  n2 = OneMinus(None=NamedRerouteUsage())
  Multiply(
   A=Subtract(A=WorldPosition().XYZ, B=ParticlePositionWS())
   B=Multiply(A=DynamicParameter() {paramNames [Size Multiplier, Param2, Param3, Param4]}.Size Multiplier, B=Saturate(None=@n2))
  ) {constB=0.5}
  n3 = ParticleColor()
  n4 = SphereMask(
   A=TextureCoordinate()
   B=Constant() {r=0.5}
   Radius=ScalarParameter() {defaultValue=0.5, parameterName=Radius}
   Hardness=ScalarParameter() {defaultValue=0.5, parameterName=Hardness}
  )
  n5 = NamedRerouteUsage()
  Multiply(A=DepthFade(Opacity=Multiply(A=@n3.A, B=@n4)) {fadeDistanceDefault=512}, B=@n5)
  n6 = ScalarParameter() {defaultValue=4000, parameterName=FadeLength}
  n7 = ScalarParameter() {defaultValue=256, parameterName=FocalPoint}
  n8 = ScalarParameter() {defaultValue=10, parameterName=FocalPointExpand}
  n9 = MaterialFunctionCall(Fade Length=@n6, Fade Offset=Add(A=@n7, B=@n8) {constB=25}) {materialFunction=&CameraDepthFade}
  n10 = Subtract(A=@n7, B=@n8)
  n11 = Multiply(A=Multiply(
    A=OneMinus(None=@n9.Result)
    B=MaterialFunctionCall(Fade Length=@n6, Fade Offset=@n10) {materialFunction=&CameraDepthFade}.Result
   )) {constB=4}
  NamedRerouteDeclaration(None=Clamp(None=@n11)) {name=TitlShift, nodeColor=<rgba 0 1 0.8359 1>}
  OneMinus(None=@n5)
  Saturate(None=@n11)
  ScalarParameter() {defaultValue=200, parameterName=SizeScaleTest}
 ]
}
