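# Masked, unlit, two-sided material for ambient KelpLinen butterflies on instanced or Niagara meshes, with switchable grass-actor pivot support.
# Emissive layers panning smoke over a color gradient, masked by the butterfly texture and hue-shifted per instance.
# World position offset combines vertex-color-selected sine wing flaps with axis rotation, varied by object height, through MF_JuiceWPOMask.
# Camera depth fade drives the opacity mask, and audio-reactive and emissive-adjustment functions post-process the attributes.
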Material /Clyde/Environments/KelpLinen/Effects/Ambient/M_KelpLinen_ButterFly_GrassActor_Scatter
graph {
 properties {
  bUseMaterialAttributes=true
  bUsedWithInstancedStaticMeshes=true
  bUsedWithNanite=true
  bUsedWithNiagaraMeshParticles=true
  blendMode=BLEND_Masked
  pixelDepthOffsetMode=PDOM_Legacy
  shadingModel=MSM_Unlit
  twoSided=true
  usedShadingModels=Unlit
 }
 nodes [
  Constant3Vector() {constant=<rgba 1 0.2 0.05 0>}
  n1 = PerInstanceRandom()
  n2 = ScalarParameter() {parameterName=HueShift}
  n3 = ScalarParameter() {defaultValue=5, parameterName=Opacity}
  n4 = Multiply(A=Subtract(A=Time(), B=Multiply(A=@n3) {constB=2}), B=ScalarParameter() {defaultValue=2, parameterName=PanSpeed})
  n5 = Panner(Coordinate=TextureCoordinate() {uTiling=3, vTiling=1.4}, Time=@n4) {speedX=-0.04, speedY=0.15}
  n6 = Panner(Coordinate=TextureCoordinate() {uTiling=1.5, vTiling=0.7}, Time=@n4) {speedX=0.05, speedY=0.25}
  n7 = VectorParameter() {parameterName=CrackColor, defaultValue=<rgba 0.1905 0.4686 0.88 0>}
  n8 = Multiply(
   A=Add(
    A=TextureSample(UVs=@n5) {texture=&T_Inky_Smoke_Tile_Chromatic}.RGB
    B=TextureSample(UVs=@n6) {texture=&T_Inky_Smoke_Tile_Chromatic}.RGB
   )
   B=Multiply(A=@n7.RGB) {constB=1.5}
  )
  n9 = TextureCoordinate() {coordinateIndex=1}
  n10 = Constant2Vector() {g=0.5, r=1}
  n11 = LinearInterpolate(
   A=VectorParameter() {parameterName=BaseColor, defaultValue=<rgba 0.1137 0.1382 0.37 0>}.RGB
   B=@n7.RGB
   Alpha=Power(Base=SphereMask(A=@n9, B=@n10) {attenuationRadius=0.5, hardnessPercent=0})
  )
  n12 = StaticSwitchParameter(True=Desaturation(None=Add(A=@n8, B=@n11), Fraction=ScalarParameter() {parameterName=Desaturation}), False=@n7.RGB) {defaultValue=true, parameterName=UseEnergyOnWings?}
  n13 = TextureSample() {texture=/Game/Effects/Fort_Effects/Textures/Bugs/T_Butterfly_Cheap_01}
  n14 = Time()
  n15 = ObjectPositionWS()
  n16 = Add(A=@n14, B=ComponentMask(None=@n15) {b=true})
  n17 = ScalarParameter() {defaultValue=-0.2, parameterName=RotationSpeed}
  n18 = ScalarParameter() {defaultValue=-0.1, parameterName=RotationSpeed_Low}
  n19 = Frac(None=ComponentMask(None=ObjectPositionWS()) {b=true})
  n20 = RotateAboutAxis(
   NormalizedRotationAxis=Transform(None=VectorParameter() {parameterName=RotationAxis, defaultValue=<rgba 0 0 1 0>}.RGB) {transformSourceType=TRANSFORMSOURCE_Local}
   RotationAngle=Multiply(A=StaticSwitchParameter(True=@n16, False=@n14) {parameterName=UsewithGrassActor?}, B=LinearInterpolate(A=@n17, B=@n18, Alpha=@n19))
   PivotPoint=StaticSwitchParameter(True=@n15, False=ActorPositionWS()) {parameterName=UsewithGrassActor?}
   Position=WorldPosition().XYZ
  )
  n21 = Normalize(VectorInput=Transform(None=Constant3Vector() {constant=<rgba 0 1 0 0>}) {transformSourceType=TRANSFORMSOURCE_ParticleWorld})
  n22 = ComponentMask(None=TextureCoordinate()) {r=true}
  n23 = Multiply(
   A=OneMinus(None=@n22)
   B=LinearInterpolate(
    A=ScalarParameter() {defaultValue=0.02, parameterName=ButterFlyWobble_Wing}
    B=ScalarParameter() {defaultValue=0.02, parameterName=ButterFlyWobble_Body}
    Alpha=@n22
   )
  )
  n24 = Time()
  n25 = ComponentMask(None=ObjectPositionWS()) {b=true}
  n26 = LinearInterpolate(
   A=ScalarParameter() {defaultValue=1, parameterName=WingPeriod}
   B=ScalarParameter() {defaultValue=1, parameterName=WingPeriod_Max}
   Alpha=Clamp(None=Frac(None=@n25))
  )
  n27 = ScalarParameter() {defaultValue=0.2, parameterName=Wing_Offset}
  n28 = Add(A=Sine(None=Multiply(A=@n24, B=@n26)), B=@n27) {constB=-0.35}
  n29 = ComponentMask(None=TextureCoordinate()) {r=true}
  n30 = ScalarParameter() {defaultValue=0.25, parameterName=Wing_Curvature}
  n31 = ScalarParameter() {defaultValue=0.2, parameterName=Wing_Amplitude}
  n32 = Multiply(A=Multiply(A=Multiply(A=@n23, B=@n28), B=Add(A=@n29, B=@n30) {constB=0.5}), B=@n31)
  n33 = MaterialFunctionCall() {materialFunction=&ObjectPivotPoint}
  n34 = WorldPosition()
  n35 = Normalize(VectorInput=Transform(None=Constant3Vector() {constant=<rgba 0 1 0 0>}) {transformSourceType=TRANSFORMSOURCE_ParticleWorld})
  n36 = Time()
  n37 = Add(A=Sine(None=Multiply(A=@n36, B=@n26) {constB=5}), B=@n27) {constB=-0.35}
  n38 = ComponentMask(None=TextureCoordinate()) {r=true}
  n39 = Multiply(A=@n31, B=Multiply(A=Multiply(A=@n23, B=@n37), B=Add(A=@n38, B=@n30) {constB=0.5}))
  n40 = MaterialFunctionCall() {materialFunction=&ObjectPivotPoint}
  n41 = WorldPosition()
  n42 = Add(
   A=Multiply(
    A=VertexColor().R
    B=RotateAboutAxis(NormalizedRotationAxis=@n21, RotationAngle=@n32, PivotPoint=@n33.Object Pivot Location, Position=@n34.XYZ)
   )
   B=Multiply(
    A=VertexColor().G
    B=RotateAboutAxis(NormalizedRotationAxis=@n35, RotationAngle=@n39, PivotPoint=@n40.Object Pivot Location, Position=@n41.XYZ)
   )
  )
  n43 = VertexColor()
  n44 = ScalarParameter() {defaultValue=100, parameterName=CamerFadeLength}
  n45 = ScalarParameter() {parameterName=CamerFadeOffset}
  n46 = Multiply(
   A=Multiply(A=LinearInterpolate(A=@n13.R, B=@n13.R, Alpha=@n43.A), B=@n3)
   B=Clamp(None=MaterialFunctionCall(Fade Length=@n44, Fade Offset=@n45) {materialFunction=&CameraDepthFade}.Result)
  )
  n47 = SetMaterialAttributes!(
   MaterialAttributes
   Emissive Color=MaterialFunctionCall(Hue Shift Percentage=Multiply(A=@n1, B=@n2), Texture=Multiply(A=@n12, B=@n13.G)) {materialFunction=&HueShift}.Result
   Opacity=@n46
   World Position Offset=MaterialFunctionCall!(WPO_Amount=ScalarParameter() {defaultValue=-5000, parameterName=JuiceWPOMask_Amount}, Position, WPO=Add(A=@n42, B=@n20)) {materialFunction=/Clyde/FrontierContent/Materials/MF_JuiceWPOMask}.Result
   Opacity Mask=@n46
  ) {attributeSetTypes [
    #1
    #2
    #3
    #4
   ]}
  MaterialFunctionCall!(In=MaterialFunctionCall!(In=@n47) {materialFunction=/ClydeCosmetics/Environments/KelpLinen/MF_Kelp_AudioReactive}.Result) {materialFunction=/Game/Effects/Niagara/MasterMaterials/Functions/MF_FX_MAEmissiveAdjustments}
  Multiply(A=@n13.G)
  ObjectPositionWS()
  ObjectPositionWS()
  OneMinus(None=@n29)
  VertexColor()
  VertexColor()
 ]
}
