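# Additive unlit material for ambient camera-facing fog puffs on instanced meshes in the KelpLinen environment, faking volume with Fresnel and depth-faded opacity.
# Opacity also scales with time of day and near-camera fade, alongside world-space noise and looping per-instance age curves.
# Emissive blends two colors by per-instance random and brightens with an audio beat tracker.
# Normal-driven displacement and a velocity term feed World Position Offset through a pivot-based Juice mask, avoiding hard edges.
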Material /Clyde/Environments/KelpLinen/Effects/Ambient/M_KelpLinen_CamFacing_VolumetricFog
graph {
 properties {
  bAutomaticallySetUsageInEditor=false
  bUseMaterialAttributes=true
  bUsedWithInstancedStaticMeshes=true
  blendMode=BLEND_Additive
  shadingModel=MSM_Unlit
  translucencyLightingMode=TLM_VolumetricPerVertexDirectional
  usedShadingModels=Unlit
 }
 nodes [
  Add(B=MaterialFunctionCall(Offset Amount=Constant() {r=500}) {materialFunction=/Engine/Functions/Engine_MaterialFunctions02/WorldPositionOffset/CameraOffset}.Camera Offset)
  Constant3Vector() {constant=<rgba 1 1 1 0>}
  Constant() {r=1}
  DepthFade(
   Opacity=Multiply(A=NamedRerouteUsage(), B=NamedRerouteUsage()) {constA=1}
   FadeDistance=ScalarParameter() {defaultValue=50, parameterName=DepthFade}
  )
  n1 = NamedRerouteUsage()
  n2 = VertexNormalWS()
  n3 = Multiply(A=@n2) {constB=0}
  Multiply(A=@n1, B=@n3)
  n4 = TransformPosition(None=WorldPosition().XYZ) {transformSourceType=TRANSFORMPOSSOURCE_World, transformType=TRANSFORMPOSSOURCE_Instance}
  n5 = Normalize(VectorInput=Subtract(A=CameraPositionWS(), B=ObjectPositionWS()))
  n6 = Normalize(VectorInput=CrossProduct(A=@n5, B=Constant3Vector() {constant=<rgba 0 0 1 0>}))
  n7 = Multiply(A=@n6) {constB=-1}
  n8 = Normalize(VectorInput=CrossProduct(A=@n6, B=@n5))
  Multiply(
   A=Subtract(
    A=MaterialFunctionCall(VectorToTransform=@n4, BasisX=@n5, BasisY=@n7, BasisZ=@n8) {materialFunction=&Transform3x3Matrix}.Transformed Vector
    B=@n4
   )
   B=MaterialFunctionCall() {materialFunction=/Engine/Functions/Engine_MaterialFunctions02/WorldPositionOffset/ObjectScale}.Scale XYZ
  )
  Multiply(
   A=VertexNormalWS()
   B=MaterialFunctionCall(Input=@n1, Input Low=Constant(), Input High=Constant() {r=1}, Target Low=Constant() {r=-500}, Target High=Constant() {r=100}) {materialFunction=&RemapValueRange}.Result
  )
  n9 = Power(Base=NamedRerouteUsage(), Exp=ScalarParameter() {defaultValue=1, parameterName=3dNoise_Mask_Power})
  n10 = ScalarParameter() {defaultValue=1, parameterName=3dNoise_Mask_Mult}
  n11 = Power(
   Base=MaterialFunctionCall(
    Brightness=ScalarParameter() {defaultValue=1, parameterName=Fresnel_brightness}
    Exponent=ScalarParameter() {defaultValue=1, parameterName=Fresnel_Exp}
    Vertex / Pixel Normal=VertexNormalWS()
   ) {materialFunction=&MF_Fresnel_FX}.Fresnel
   Exp=LinearInterpolate(B=ScalarParameter() {defaultValue=10, parameterName=Fresnel_WPO_Power}, Alpha=NamedRerouteUsage()) {constB=4}
  )
  n12 = ScalarParameter() {defaultValue=2, parameterName=Fresnel_Mult}
  n13 = PixelDepth()
  n14 = ScalarParameter() {defaultValue=650, parameterName=Param_1}
  n15 = Multiply(A=Clamp(None=Multiply(A=@n11, B=@n12)), B=Saturate(None=Divide(A=@n13.R, B=@n14)))
  Multiply(A=Clamp(None=Multiply(A=@n9, B=@n10)), B=@n15)
  n16 = TextureSample(UVs=Multiply(A=WorldPosition().XYZ, B=ScalarParameter() {defaultValue=0.0034, parameterName=Param_2})) {samplerType=SAMPLERTYPE_LinearColor, texture=/Game/Global/MaterialFunctions/VT_ClothSimWPO}
  NamedRerouteDeclaration(None=@n16.RGB) {name=Noise_WPO, nodeColor=<rgba 0 0.7188 1 1>}
  NamedRerouteDeclaration(None=@n16.RGB) {name=Normal, nodeColor=<rgba 0 1 0.5547 1>}
  n17 = WorldPosition()
  n18 = ScalarParameter() {defaultValue=0.0034, parameterName=Param_2}
  n19 = Time()
  n20 = Constant3Vector() {constant=<rgba 0.1 0 0.025 0>}
  n21 = TextureSample(UVs=Add(A=Multiply(A=@n17.XYZ, B=@n18), B=Multiply(A=@n19, B=@n20))) {samplerType=SAMPLERTYPE_LinearColor, texture=&VT_ClothSimWPO}
  NamedRerouteDeclaration(None=@n21.RGB) {name=Normal_2, nodeColor=<rgba 0 1 0.5547 1>}
  n22 = MaterialFunctionCall!(
   WorldPosition=WorldPosition().XYZ
   PivotLocation=MaterialFunctionCall() {materialFunction=&ComponentPivotLocation}.Result
   WS Normals=Transform(None=VertexNormalWS()) {transformSourceType=TRANSFORMSOURCE_World, transformType=TRANSFORM_Instance}
   Custom Object Basis 1=Transform(None=VectorParameter() {parameterName=X, defaultValue=<rgba 0 0 -1 1>}.RGB) {transformSourceType=TRANSFORMSOURCE_World}
   Custom Object Basis 2=Transform(None=VectorParameter() {parameterName=Y, defaultValue=<rgba -1 0 0 1>}.RGB) {transformSourceType=TRANSFORMSOURCE_World}
   Custom Object Basis 3=Transform(None=VectorParameter() {parameterName=Z, defaultValue=<rgba 0 1 0 1>}.RGB) {transformSourceType=TRANSFORMSOURCE_World}
  ) {materialFunction=/Engine/Functions/Engine_MaterialFunctions02/WorldPositionOffset/AlignMeshToTheCamera}
  NamedRerouteDeclaration(None=@n22.Rotated Normals) {name=AlignMesh_Normal, nodeColor=<rgba 0.6641 1 0 1>}
  NamedRerouteDeclaration(None=@n22.World Position Offset) {name=AlignMesh_WPO, nodeColor=<rgba 1 0 0.9844 1>}
  n23 = PerInstanceRandom()
  n24 = Multiply(A=Time(), B=ScalarParameter() {defaultValue=0.05, parameterName=AgeMultiplier})
  n25 = PerInstanceRandom()
  NamedRerouteDeclaration(None=Modulo!(A=Add(A=@n23, B=@n24), B=LinearInterpolate(Alpha=@n25) {constA=1, constB=2})) {name=NormalizedAge, nodeColor=<rgba 0.125 1 0 1>}
  NamedRerouteDeclaration(None=Sine(None=NamedRerouteUsage())) {name=FadeByAge, nodeColor=<rgba 0 0.3672 1 1>}
  n26 = Normalize(VectorInput=Subtract(A=CameraPositionWS(), B=ObjectPositionWS()))
  n27 = Normalize(VectorInput=CrossProduct(A=@n26, B=Constant3Vector() {constant=<rgba 0 0 1 0>}))
  n28 = CrossProduct(A=@n27, B=@n26)
  NamedRerouteDeclaration(None=MaterialFunctionCall(VectorToTransform=VertexNormalWS(), BasisX=@n26, BasisY=Multiply(A=@n27) {constB=-1}, BasisZ=Normalize(VectorInput=@n28)) {materialFunction=&Transform3x3Matrix}.Transformed Vector) {name=WPO_Normal, nodeColor=<rgba 0.8984 1 0 1>}
  NamedRerouteDeclaration(None=SphereMask(A=NamedRerouteUsage(), B=ScalarParameter() {defaultValue=0.5, parameterName=NormalizeAge_B}) {attenuationRadius=0.5, hardnessPercent=0}) {name=LifetimeCurve, nodeColor=<rgba 1 0.7734 0 1>}
  n29 = TextureSample(UVs=Multiply(A=WorldPosition().XYZ, B=ScalarParameter() {defaultValue=0.0034, parameterName=Param_2})) {samplerType=SAMPLERTYPE_LinearColor, texture=&VT_ClothSimWPO}
  NamedRerouteDeclaration(None=@n29.RGB) {name=NewNoiseWPO, nodeColor=<rgba 0 0.7188 1 1>}
  NamedRerouteDeclaration(None=@n29.RGB) {name=Normal_1, nodeColor=<rgba 0 1 0.5547 1>}
  n30 = Multiply(A=@n29.RGB, B=ScalarParameter() {defaultValue=0.222, parameterName=Param_3})
  n31 = Saturate(None=DotProduct(A=CameraVectorWS(), B=NamedRerouteUsage()))
  NamedRerouteDeclaration(None=SphereMask(A=Add(A=@n30, B=@n31), B=ScalarParameter() {defaultValue=0.9664, parameterName=Param_4}, Radius=ScalarParameter() {defaultValue=0.4}) {hardnessPercent=0}) {name=NewNoise, nodeColor=<rgba 0.5938 0 1 1>}
  NamedRerouteDeclaration(None=@n21.RGB) {name=WPO, nodeColor=<rgba 0 0.7188 1 1>}
  NamedRerouteUsage()
  NamedRerouteUsage()
  NamedRerouteUsage()
  NamedRerouteUsage()
  NamedRerouteUsage()
  NamedRerouteUsage()
  NamedRerouteUsage()
  OneMinus()
  n32 = VertexNormalWS()
  n33 = MaterialFunctionCall(
   Normal=Transform(None=MaterialFunctionCall!(Bump=@n32) {materialFunction=/Engine/Functions/Engine_MaterialFunctions03/Procedurals/PerturbNormalLQ}.WS Normal) {transformSourceType=TRANSFORMSOURCE_World, transformType=TRANSFORM_Tangent}
   Flatness=ScalarParameter() {parameterName=Param_7}
  ) {materialFunction=&FlattenNormal}
  n34 = Constant()
  QualitySwitch(Default=@n33.Result, Low=@n34, High=@n34, Medium=@n34, Epic=@n33.Result)
  n35 = ScalarParameter() {defaultValue=1, parameterName=Reactive_Low}
  n36 = ScalarParameter() {defaultValue=1.5, parameterName=Reactive_High}
  n37 = Clamp(None=MaterialFunctionCall() {materialFunction=/ClydeCosmetics/Environments/KelpLinen/MF_Kelp_AudioReactive_Parameter}.Beat_Tracker)
  n38 = VectorParameter() {parameterName=Color_A, defaultValue=<rgba 0 0.9489 1 1>}
  n39 = VectorParameter() {parameterName=Color_B, defaultValue=<rgba 0.5841 0 1 1>}
  n40 = PerInstanceRandom()
  n41 = MaterialFunctionCall() {
   desc="Absolute world pos causing Hard edges at certain distance, \r\nchanging it to pivot will prevent the hard edges by moving the\r\n object as a whole instead of harshly moving half of it down(when i saw it)"
   materialFunction=&ObjectPivotPoint
  }
  n42 = NamedRerouteUsage()
  n43 = VertexNormalWS()
  n44 = ScalarParameter() {defaultValue=1, parameterName=WPO Amount}
  n45 = ScalarParameter() {defaultValue=1, parameterName=WPO Amount_Max}
  n46 = PerInstanceRandom()
  n47 = Add(
   A=Multiply(A=Multiply(A=@n42, B=@n43), B=LinearInterpolate(A=@n44, B=@n45, Alpha=@n46))
   B=LinearInterpolate(A=Multiply(A=@n2) {constB=0}, B=@n3, Alpha=NamedRerouteUsage())
  )
  n48 = Multiply(A=NamedRerouteUsage(), B=VectorParameter() {parameterName=Velocity, defaultValue=<rgba 0 0 0 1>}.RGB)
  n49 = ScalarParameter() {defaultValue=250, parameterName=Param_6}
  n50 = Multiply(A=Multiply(A=NamedRerouteUsage(), B=DepthFade(Opacity=@n15, FadeDistance=@n49)), B=NamedRerouteUsage())
  n51 = MaterialFunctionCall(
   Color_Morning=ScalarParameter() {defaultValue=0.4, parameterName=Opacity_Morning}
   Color_Day=ScalarParameter() {defaultValue=0.4, parameterName=Opacity_Day}
   Color_Evening=ScalarParameter() {defaultValue=0.4, parameterName=Opacity_Evening}
   Color_Night=ScalarParameter() {defaultValue=0.4, parameterName=Opacity_Night}
  ) {materialFunction=&Color_Time_of_Day}
  n52 = MaterialFunctionCall(Distance=ScalarParameter() {defaultValue=1000, parameterName=NearCameraFade}) {materialFunction=&Near_Camera_Fade}
  SetMaterialAttributes!(
   MaterialAttributes
   Emissive Color=Multiply(A=LinearInterpolate(A=@n35, B=@n36, Alpha=@n37) {constA=1}, B=LinearInterpolate(A=@n38.RGB, B=@n39.RGB, Alpha=@n40))
   Opacity=Multiply(A=Multiply(A=@n50, B=@n51.Result), B=OneMinus(None=@n52.Result))
   World Position Offset=MaterialFunctionCall!(WPO_Amount=Constant() {r=-10000}, Position=@n41.Object Pivot Location, WPO=Add(A=@n47, B=@n48)) {materialFunction=/Clyde/FrontierContent/Materials/MF_JuiceWPOMask}.Result
  ) {attributeSetTypes [#1, #2, #3]}
  n53 = NamedRerouteUsage()
  n54 = Multiply(
   A=Power(Base=ComponentMask(None=@n53) {g=true, r=true}, Exp=ScalarParameter() {defaultValue=1, parameterName=Noise_Distortion_Power})
   B=ScalarParameter() {defaultValue=0.5, parameterName=Noise_Distortion_Intensity}
  ) {constB=0.2}
  n55 = TextureCoordinate()
  SphereMask(A=ComponentMask(None=Add(A=@n54, B=@n55)) {g=true}, B=Constant() {r=1}, Radius=Constant() {r=1}) {attenuationRadius=100, hardnessPercent=0}
  n56 = ScalarParameter() {defaultValue=0.222, parameterName=Param_3}
  n57 = DotProduct(A=CameraVectorWS(), B=NamedRerouteUsage())
  SphereMask(
   A=Add(A=Multiply(A=@n21.RGB, B=@n56), B=Saturate(None=@n57))
   B=ScalarParameter() {defaultValue=0.9664, parameterName=Param_4}
   Radius=ScalarParameter() {defaultValue=0.4}
  ) {hardnessPercent=0}
  Transform() {transformSourceType=TRANSFORMSOURCE_Instance}
  Transform(None=ObjectPositionWS()) {transformSourceType=TRANSFORMSOURCE_World, transformType=TRANSFORM_Instance}
  VectorParameter() {parameterName=WPO_Scale_Max, defaultValue=<rgba 500 500 0 1>}
  VertexNormalWS()
  VertexNormalWS()
  VertexNormalWS()
 ]
}
