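# Translucent unlit material for ambient bokeh motes in the KelpLinen environment, built for instanced static meshes.
# Vertex colors randomize each sprite's looping time-based age, size, velocity and camera offset, animating motes entirely through world position offset with a JuiceWPOMask.
# Bokeh texture opacity combines lifetime, depth, near-camera and instance fades, with near-camera distance also shaping a soft mask.
# Per-instance hue shift colors the emissive before MF_Kelp_AudioReactive processes the attributes.
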Material /Clyde/Environments/KelpLinen/Effects/Ambient/M_KelpLinen_NoTick_BokehMotes
graph {
 properties {
  bAutomaticallySetUsageInEditor=false
  bTangentSpaceNormal=false
  bUseMaterialAttributes=true
  bUsedWithInstancedStaticMeshes=true
  blendMode=BLEND_Translucent
  shadingModel=MSM_Unlit
  twoSided=true
  usedShadingModels=Unlit
 }
 nodes [
  n1 = WorldPosition()
  n2 = TextureCoordinate()
  n3 = ScalarParameter() {defaultValue=30, parameterName=Scale}
  n4 = ScalarParameter() {defaultValue=60, parameterName=Scale Max}
  n5 = VertexColor()
  n6 = ScalarParameter() {defaultValue=1, parameterName=Scale over life Min}
  n7 = ScalarParameter() {defaultValue=4, parameterName=Scale Over Life Max}
  n8 = NamedRerouteUsage()
  n9 = Multiply(A=Multiply(A=LinearInterpolate(A=@n3, B=@n4, Alpha=@n5.G), B=LinearInterpolate(A=@n6, B=@n7, Alpha=@n8)), B=NamedRerouteUsage())
  n10 = VertexColor()
  n11 = Time()
  n12 = LinearInterpolate(Alpha=VertexColor().B)
  n13 = Multiply(A=Add(A=Add(A=@n11, B=@n12), B=NamedRerouteUsage()), B=ScalarParameter() {defaultValue=0.1, parameterName=Delta Time Scale})
  n14 = Multiply(A=Fmod(A=Add(A=@n10.A, B=@n13), B=ScalarParameter() {defaultValue=1, parameterName=Age}), B=NamedRerouteUsage())
  n15 = LinearInterpolate(
   A=VectorParameter() {parameterName=Velocity Min, defaultValue=<rgba 200 0 0 0>}.RGB
   B=VectorParameter() {parameterName=VelocityMax, defaultValue=<rgba 200 0 100 0>}.RGB
   Alpha=VertexColor().G
  )
  n16 = ScalarParameter() {defaultValue=75, parameterName=Camera Offset Min}
  n17 = ScalarParameter() {defaultValue=150, parameterName=Camera Offset}
  n18 = VertexColor()
  n19 = Add(
   A=Add(
    A=MaterialFunctionCall!(Center Location=@n1.XYZ, 0-1 UVs=@n2, XY scale=@n9, OverridePivot, "Normals (Optional)") {materialFunction=&Sprite}.World position offset
    B=Multiply(A=@n14, B=@n15)
   )
   B=MaterialFunctionCall(Offset Amount=LinearInterpolate(A=@n16, B=@n17, Alpha=@n18.R), World Position=WorldPosition().XYZ) {materialFunction=/Engine/Functions/Engine_MaterialFunctions02/WorldPositionOffset/CameraOffset}.Camera Offset
  )
  n20 = VectorParameter() {parameterName=LocalOffset, defaultValue=<rgba 0 0 500 1>}
  n21 = WorldPosition()
  Add(
   A=@n19
   B=TransformPosition(None=Add(A=@n20.RGB, B=@n21.XYZ)) {transformSourceType=TRANSFORMPOSSOURCE_World, transformType=TRANSFORMPOSSOURCE_Instance}
  )
  n22 = LinearInterpolate(B=ScalarParameter() {defaultValue=0.7, parameterName=Hue Shift Percentage}, Alpha=Multiply(A=PerInstanceRandom(), B=VertexColor().B))
  n23 = VectorParameter() {parameterName=Emissive_Color_A, defaultValue=<rgba 0 0.939 1 1>}
  n24 = ScalarParameter() {defaultValue=-5000, parameterName=JuiceWPOMask_Amount}
  n25 = SphereMask(
   A=TextureCoordinate()
   B=Constant() {r=0.5}
   Radius=MaterialFunctionCall(Distance=ScalarParameter() {defaultValue=1000, parameterName=NearCamera_Blur_Distance}) {materialFunction=&Near_Camera_Fade}.Result
  ) {attenuationRadius=1, hardnessPercent=0}
  n26 = TextureSample() {texture=/Engine/Engine_MI_Shaders/Textures/Bokeh}
  n27 = ScalarParameter() {defaultValue=0.1, parameterName=Overall Opacity}
  n28 = NamedRerouteUsage()
  n29 = ScalarParameter() {defaultValue=0.25, parameterName=Opacity Fade Start}
  n30 = ScalarParameter() {defaultValue=0.25, parameterName=Radius}
  n31 = Multiply(
   A=Multiply(A=Multiply(A=@n26.RGB, B=@n27), B=SphereMask(A=@n28, B=@n29, Radius=@n30) {attenuationRadius=0.1, hardnessPercent=0})
   B=NamedRerouteUsage()
  )
  n32 = ScalarParameter() {defaultValue=50, parameterName=Depth Fade}
  n33 = Multiply(
   A=Multiply(A=PerInstanceFadeAmount(), B=DepthFade(Opacity=@n31, FadeDistance=@n32))
   B=MaterialFunctionCall(Distance=ScalarParameter() {parameterName=NearCameraFade_Distance}) {materialFunction=&Near_Camera_Fade}.Result
  )
  MaterialFunctionCall!(In=SetMaterialAttributes!(
    MaterialAttributes
    Emissive Color=MaterialFunctionCall(Hue Shift Percentage=@n22, Texture=@n23.RGB) {materialFunction=&HueShift}.Result
    Opacity=Multiply(A=@n25, B=@n33)
    World Position Offset=MaterialFunctionCall!(WPO_Amount=@n24, Position, WPO=@n19) {materialFunction=/Clyde/FrontierContent/Materials/MF_JuiceWPOMask}.Result
   ) {attributeSetTypes [#1, #2, #3]}) {materialFunction=/ClydeCosmetics/Environments/KelpLinen/MF_Kelp_AudioReactive}
  NamedRerouteDeclaration(None=@n14) {name=Age, nodeColor=<rgba 0 0.5078 1 1>}
  NamedRerouteDeclaration() {name=RecalculatedNormals, nodeColor=<rgba 0.9219 1 0 1>}
  NamedRerouteDeclaration(None=MaterialFunctionCall(
    Value_Morning=ScalarParameter() {defaultValue=1, parameterName=Opacity Morning}
    Value_Day=ScalarParameter() {defaultValue=1, parameterName=Opacity Day}
    Value_Evening=ScalarParameter() {defaultValue=1, parameterName=Opacity Evening}
    Value_Night=ScalarParameter() {defaultValue=1, parameterName=Opacity Night}
   ) {materialFunction=&Scalar_Time_of_Day}.Result) {name=TOD Opacity, nodeColor=<rgba 0.5469 0 1 1>}
  NamedRerouteDeclaration(None=Step(Y=VertexColor().A, X=ScalarParameter() {defaultValue=0.9, parameterName=Cull})) {name=Spawn Rate, nodeColor=<rgba 1 0.5391 0 1>}
  n34 = ObjectPositionWS()
  n35 = ScalarParameter() {defaultValue=500, parameterName=Randomness}
  n36 = Frac(None=Divide(A=ComponentMask(None=@n34) {r=true}, B=@n35))
  n37 = Frac(None=Divide(A=ComponentMask(None=@n34) {g=true}, B=@n35))
  n38 = Divide(A=ComponentMask(None=@n34) {b=true}, B=@n35)
  NamedRerouteDeclaration(None=Add(A=Add(A=@n36, B=@n37), B=Frac(None=@n38))) {name=Randomness, nodeColor=<rgba 0 1 0.1328 1>}
  VertexColor()
 ]
}
