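# Niagara ambient effect for the KelpLinen environment: floating fake-bokeh motes on GPU sprites, spawned continuously through a wide, shallow box.
# Each mote gets a random curve-picked tint, a random size and a randomized dynamic material parameter.
# Wind and aerodynamic drag drift the motes, while an alpha curve scaled by a waveform fades them in and out with varied intensity.
# Only NOTILT is enabled; HangingParticulates is a disabled smaller-sprite variant.
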NiagaraSystem /Clyde/Environments/KelpLinen/Effects/Ambient/NS_KelpLinen_BokehMote_FakeBokeh_Floating
dependencies {
 usedDataInterfaces [
  &NiagaraDataInterfaceColorCurve
  &NiagaraDataInterfaceCollisionQuery
  &NiagaraDataInterfaceLandscape
  &NiagaraDataInterfaceCurve
 ]
 usedDynamicInputs []
 usedModules [&SystemState]
}
emitters {
 HangingParticulates {
  bEnabled=false
  simTarget=GPUComputeSim
  emitterUpdateScript [
   &EmitterState(Life Cycle Mode=::NewEnumerator1, Loop Duration=2, UseLoopDelay=false)
   &SpawnRate(SpawnRate=120, Use Spawn Probability=false)
  ]
  particleSpawnScript [
   &InitializeParticle(
    Color=&ColorFromCurve(
     ColorCurve=DataInterface() {propertyValues {
       AlphaCurve {keys [<at 0.4783 0.15 RCIM_Cubic>]}
       BlueCurve {keys [
         <at 0.017 0.2916>
         {arriveTangent=0.4465, interpMode=RCIM_Cubic, leaveTangent=0.4465, time=0.3299, value=0.7863}
         <at 0.6588 0.5781>
         <at 1 0.7482 RCIM_Cubic>
        ]}
       GreenCurve {keys [
         <at 0.017 0.4935>
         {arriveTangent=-0.4734, interpMode=RCIM_Cubic, leaveTangent=-0.4734, time=0.3299, value=0.3424}
         <at 0.6588 0.1897>
         <at 1 1 RCIM_Cubic>
        ]}
       RedCurve {keys [
         <at 0.017 1>
         {arriveTangent=-0.9512, interpMode=RCIM_Cubic, leaveTangent=-0.9512, time=0.3299, value=1}
         <at 0.6588 0.3895>
         <at 1 0.037 RCIM_Cubic>
        ]}
      }}
     CurveIndex=&RandomRangeFloat(Minimum=0, Maximum=1)
    )
    Lifetime Max=8
    Lifetime Min=5
    Lifetime Mode=::NewEnumerator1
    Sprite Rotation Mode=::NewEnumerator2
    Sprite Size Mode=::NewEnumerator1
    Sprite UV Mode=::NewEnumerator3
    Uniform Sprite Size Max=12
    UsePositionOffset=false
    Write Material Random=true
   )
   &ShapeLocation(
    Box Midpoint=&RandomRangeVector(Minimum=<0 0 0>, Maximum=<1 1 1>)
    Box Size=<6500 6500 400>
    MaskSpawnGroup=false
    Override Seed=false
    Shape Primitive=::NewEnumerator2
    Surface Only=false
    UseOwnerScale=true
   )
   &DynamicMaterialParameters(
    Index 0 Param 1=&RandomRangeFloat(Minimum=25, Maximum=100)
    Module.Index0Param0Write=true
    Module.Index0Param1Write=true
    Module.Index0Param2Write=true
    Module.Index0Param3Write=true
   )
  ]
  particleUpdateScript [
   &ParticleState()
   &WindForce(Frequency=6, Offset By Time:Vector3f=<0 0 -50>, Scale=1, Wind Speed=<1 0 0>, Wind Speed Scale=25)
   &AerodynamicDrag(
    Aerodynamic Drag:NiagaraFloat=&RandomRangeFloat(Minimum=0.6, Maximum=1)
    Aerodynamic Rotational Drag=0.2
    Drag Shape Curve=DataInterface() {propertyValues {Curve {keys [
        {interpMode=RCIM_Cubic, tangentMode=RCTM_User, time=0, value=0.1}
        {arriveTangent=2.0799, interpMode=RCIM_Cubic, leaveTangent=2.0799, tangentMode=RCTM_User, time=1, value=1}
       ]}}}
    Lift Shape Curve=DataInterface() {propertyValues {Curve {keys [
        {interpMode=RCIM_Cubic, tangentMode=RCTM_User, time=0, value=0}
        {interpMode=RCIM_Cubic, tangentMode=RCTM_User, time=0.5, value=1}
        {tangentMode=RCTM_User, time=1, value=0}
       ]}}}
    Use Linear Drag=true
    Use Rotational Drag=true
   )
   &ScaleSpriteSize(Scale Factor=&Vector2DFromFloat(Value=&FloatFromCurve(FloatCurve=DataInterface() {propertyValues {Curve {keys [
          <at 0 0 RCIM_Cubic>
          <at 0.112 1.0173>
          {arriveTangent=-1.142, interpMode=RCIM_Cubic, leaveTangent=-1.142, time=0.7001, value=0.872}
          <at 1.0011 0.002>
         ]}}}))) {enabled=false}
   &ScaleColor(
    Scale Alpha=&FloatFromCurve(
     FloatCurve=DataInterface() {propertyValues {Curve {keys [<at 0 0 RCIM_Cubic>,<at 0.5 1 RCIM_Cubic>,<at 1 0 RCIM_Cubic>]}}}
     Scale Curve=&Multiply_Float(
      A=1
      B=&Waveform(
       Global Input Position=&RandomRangeFloat(Minimum=0, Maximum=1)
       Global Clamp Min=0.5
       "Waveform [1]"=::NewEnumerator9
       "[1] Frequency"=5
       "[1] Amplitude Min/Max"=<0 1>
       "[1] Scale"=0.5
       "[1] Bias"=1
      )
     )
    )
    ScaleA=true
    ScaleRGB=true
   )
   &SolveForcesAndVelocity(Clamp Velocity=false, Limit Acceleration=false)
  ]
  rendererClasses [&NiagaraSpriteRendererProperties]
  renderers [{rendererClass=&NiagaraSpriteRendererProperties}]
 }
 NOTILT {
  simTarget=GPUComputeSim
  emitterUpdateScript [
   &EmitterState(Life Cycle Mode=::NewEnumerator1, Loop Duration=2, UseLoopDelay=false)
   &SpawnRate(SpawnRate=120, Use Spawn Probability=false)
  ]
  particleSpawnScript [
   &InitializeParticle(
    Color=&ColorFromCurve(
     ColorCurve=DataInterface() {propertyValues {
       AlphaCurve {keys [<at 0.4783 0.15 RCIM_Cubic>]}
       BlueCurve {keys [
         <at 0.017 0.2916>
         {arriveTangent=0.4465, interpMode=RCIM_Cubic, leaveTangent=0.4465, time=0.3299, value=0.7863}
         <at 0.6588 0.5781>
         <at 1 0.7482 RCIM_Cubic>
        ]}
       GreenCurve {keys [
         <at 0.017 0.4935>
         {arriveTangent=-0.4734, interpMode=RCIM_Cubic, leaveTangent=-0.4734, time=0.3299, value=0.3424}
         <at 0.6588 0.1897>
         <at 1 1 RCIM_Cubic>
        ]}
       RedCurve {keys [
         <at 0.017 1>
         {arriveTangent=-0.9512, interpMode=RCIM_Cubic, leaveTangent=-0.9512, time=0.3299, value=1}
         <at 0.6588 0.3895>
         <at 1 0.037 RCIM_Cubic>
        ]}
      }}
     CurveIndex=&RandomRangeFloat(Minimum=0, Maximum=1)
    )
    Lifetime Max=8
    Lifetime Min=5
    Lifetime Mode=::NewEnumerator1
    Sprite Rotation Mode=::NewEnumerator2
    Sprite Size Mode=::NewEnumerator1
    Sprite UV Mode=::NewEnumerator3
    Uniform Sprite Size Max=75
    UsePositionOffset=false
    Write Material Random=true
   )
   &ShapeLocation(
    Box Midpoint=&RandomRangeVector(Minimum=<0 0 0>, Maximum=<1 1 1>)
    Box Size=<6500 6500 400>
    MaskSpawnGroup=false
    Override Seed=false
    Shape Primitive=::NewEnumerator2
    Surface Only=false
    UseOwnerScale=true
   )
   &DynamicMaterialParameters(
    Index 0 Param 1=&RandomRangeFloat(Minimum=2, Maximum=12)
    Module.Index0Param0Write=true
    Module.Index0Param1Write=true
    Module.Index0Param2Write=true
    Module.Index0Param3Write=true
   )
  ]
  particleUpdateScript [
   &ParticleState()
   &WindForce(Frequency=6, Offset By Time:Vector3f=<0 0 -50>, Scale=1, Wind Speed=<1 0 0>, Wind Speed Scale=25)
   &AerodynamicDrag(
    Aerodynamic Drag:NiagaraFloat=&RandomRangeFloat(Minimum=0.6, Maximum=1)
    Aerodynamic Rotational Drag=0.2
    Drag Shape Curve=DataInterface() {propertyValues {Curve {keys [
        {interpMode=RCIM_Cubic, tangentMode=RCTM_User, time=0, value=0.1}
        {arriveTangent=2.0799, interpMode=RCIM_Cubic, leaveTangent=2.0799, tangentMode=RCTM_User, time=1, value=1}
       ]}}}
    Lift Shape Curve=DataInterface() {propertyValues {Curve {keys [
        {interpMode=RCIM_Cubic, tangentMode=RCTM_User, time=0, value=0}
        {interpMode=RCIM_Cubic, tangentMode=RCTM_User, time=0.5, value=1}
        {tangentMode=RCTM_User, time=1, value=0}
       ]}}}
    Use Linear Drag=true
    Use Rotational Drag=true
   )
   &ScaleSpriteSize(Scale Factor=&Vector2DFromFloat(Value=&FloatFromCurve(FloatCurve=DataInterface() {propertyValues {Curve {keys [
          <at 0 0 RCIM_Cubic>
          <at 0.112 1.0173>
          {arriveTangent=-1.142, interpMode=RCIM_Cubic, leaveTangent=-1.142, time=0.7001, value=0.872}
          <at 1.0011 0.002>
         ]}}}))) {enabled=false}
   &ScaleColor(
    Scale Alpha=&FloatFromCurve(
     FloatCurve=DataInterface() {propertyValues {Curve {keys [<at 0 0 RCIM_Cubic>,<at 0.5 1 RCIM_Cubic>,<at 1 0 RCIM_Cubic>]}}}
     Scale Curve=&Multiply_Float(
      A=1
      B=&Waveform(
       Global Input Position=&RandomRangeFloat(Minimum=0, Maximum=1)
       Global Clamp Min=0.5
       "Waveform [1]"=::NewEnumerator9
       "[1] Frequency"=5
       "[1] Amplitude Min/Max"=<0 1>
       "[1] Scale"=0.5
       "[1] Bias"=1
      )
     )
    )
    ScaleA=true
    ScaleRGB=true
   )
   &SolveForcesAndVelocity(Clamp Velocity=false, Limit Acceleration=false)
  ]
  rendererClasses [&NiagaraSpriteRendererProperties]
  renderers [{rendererClass=&NiagaraSpriteRendererProperties}]
 }
}
summary {
 systemName=NS_KelpLinen_BokehMote_FakeBokeh_Floating
 emitters [
  {bEnabled=false, emitterName=HangingParticulates, simTarget=GPUComputeSim, rendererClasses [&NiagaraSpriteRendererProperties]}
  {emitterName=NOTILT, simTarget=GPUComputeSim, rendererClasses [&NiagaraSpriteRendererProperties]}
 ]
 userVariables []
}
